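# Default Lit material for album cover icons under Athena UI, with parameterized cover art that defaults to a Season 6 music texture.
# The graph builds a vertical gradient between two color parameters (green by default), using the texture V coordinate remapped to start at 0.25.
# It also holds the CoverArtImage texture parameter, a fixed album cover normal map, and low Specular and Roughness scalar parameters for tuning the surface.
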Material /Game/Athena/UI/Icons/M_Album_Cover
graph {
 properties {pixelDepthOffsetMode=PDOM_Legacy, usedShadingModels=Default Lit}
 nodes [
  n1 = TextureCoordinate()
  LinearInterpolate(
   A=VectorParameter() {parameterName=RGBA1, defaultValue=<rgba 0.0414 0.22 0.0242 0>}.RGB
   B=VectorParameter() {parameterName=RGBA0, defaultValue=<rgba 0.203 0.535 0.0642 0>}.RGB
   Alpha=LinearInterpolate(Alpha=ComponentMask(None=@n1) {g=true}) {constA=0.25}
  )
  ScalarParameter() {defaultValue=0.03, parameterName=Specular}
  ScalarParameter() {defaultValue=0.0238, parameterName=Roughness}
  TextureSampleParameter2D() {parameterName=CoverArtImage, texture=/Game/2dAssets/Music/Season6/Textures/T_Music_OGRemix}
  TextureSample() {samplerType=SAMPLERTYPE_Normal, texture=&/T_Album_Cover_N}
 ]
}
