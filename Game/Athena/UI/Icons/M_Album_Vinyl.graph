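# M_Album_Vinyl is a Default Lit material in the Athena UI icons folder, named for a vinyl album.
# It holds two texture samples, the T_Album_Vinyl_D diffuse map and the T_Album_Vinyl_N normal map, with no parameters or math nodes.
# The graph shows no links from either sample to a material output.
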Material /Game/Athena/UI/Icons/M_Album_Vinyl
graph {
 properties {pixelDepthOffsetMode=PDOM_Legacy, usedShadingModels=Default Lit}
 nodes [
  TextureSample() {samplerType=SAMPLERTYPE_Normal, texture=&/T_Album_Vinyl_N}
  TextureSample() {texture=&/T_Album_Vinyl_D}
 ]
}
