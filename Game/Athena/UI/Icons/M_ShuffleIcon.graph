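# Translucent unlit UI material for the shuffle icon mesh, giving it an animated, glowing, bobbing look.
# UV-band masks choose between face and edge colour parameters and between two panning sine stripe patterns that modulate the colour.
# The stripes also drive an opacity-style mask, with a SmoothStep sine sweep and trimmed UV borders.
# A vertex-normal-based vector adds time-driven Z motion, phase-shifted between the two UV halves.
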Material /Game/Athena/UI/Icons/M_ShuffleIcon
graph {
 properties {
  blendMode=BLEND_Translucent
  pixelDepthOffsetMode=PDOM_Legacy
  previewMesh=&/S_UI_ShuffleIcon
  shadingModel=MSM_Unlit
  usedShadingModels=Unlit
 }
 nodes [
  n1 = Frac(None=TextureCoordinate() {vTiling=2})
  n2 = Ceil(None=ComponentMask(None=Add(A=@n1) {constB=-0.5}) {g=true})
  n3 = TextureCoordinate() {coordinateIndex=1}
  n4 = Time()
  n5 = Abs(None=Sine(None=Panner(Coordinate=@n3, Time=@n4) {speedY=-0.0125}) {period=0.05})
  n6 = TextureCoordinate() {uTiling=0.6184}
  n7 = Panner(Coordinate=Rotator(Coordinate=@n6, Time=Constant() {r=0.75}) {speed=6.2832}) {speedX=-0.0125}
  n8 = Power(Base=Abs(None=Sine(None=@n7) {period=0.025})) {constExponent=4}
  n9 = LinearInterpolate(A=ComponentMask(None=Power(Base=@n5) {constExponent=1.5}) {g=true}, B=Add(A=ComponentMask(None=@n8) {r=true}) {constB=0.25}, Alpha=@n2)
  LinearInterpolate(
   B=LinearInterpolate(
    A=VectorParameter() {parameterName=FaceColor, defaultValue=<rgba 0 1.724 16 1>}.RGB
    B=VectorParameter() {parameterName=EdgeColor, defaultValue=<rgba 0 6.896 32 1>}.RGB
    Alpha=@n2
   )
   Alpha=@n9
  )
  n10 = Abs(None=Sine(None=Panner(Coordinate=@n6, Time=@n4) {speedX=-0.25}) {period=2})
  n11 = MaterialFunctionCall(Alpha=ComponentMask(None=Power(Base=@n10) {constExponent=16}) {r=true}, Min=Constant() {r=0.125}, Max=Constant() {r=1}) {materialFunction=&SmoothStep}
  n12 = TextureCoordinate()
  n13 = Ceil(None=ComponentMask(None=Add(A=@n12) {constB=-0.5}) {g=true})
  n14 = Add(A=TextureCoordinate()) {constB=-0.025}
  n15 = Ceil(None=ComponentMask(None=@n14) {r=true})
  n16 = Ceil(None=ComponentMask(None=@n14) {g=true})
  LinearInterpolate(
   B=Multiply(A=LinearInterpolate(B=@n9, Alpha=@n11.Result) {constA=0.075}, B=Clamp(None=@n13) {minDefault=0.175})
   Alpha=Clamp(None=LinearInterpolate(A=@n15, Alpha=@n16))
  )
  n17 = MaterialFunctionCall(Float3=VertexNormalWS()) {materialFunction=&BreakOutFloat3Components}
  n18 = Constant() {r=8}
  n19 = Constant() {r=16}
  n20 = Time()
  n21 = LinearInterpolate(A=@n18, B=@n19, Alpha=Sine(None=@n20) {period=4}) {constA=5, constB=10}
  n22 = LinearInterpolate(A=@n18, B=@n19, Alpha=Sine(None=Add(A=@n20) {constB=-0.25}) {period=4}) {constA=5, constB=10}
  MaterialFunctionCall(X=@n17.R, Y=@n17.R, Z=Add(A=@n17.B, B=LinearInterpolate(A=@n21, B=@n22, Alpha=@n13))) {materialFunction=&MakeFloat3}
  TextureCoordinate() {uTiling=0.6184}
 ]
}
